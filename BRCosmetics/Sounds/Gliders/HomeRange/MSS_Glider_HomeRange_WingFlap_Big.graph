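# Stereo one-shot MetaSound source for the big wing flap sound of the HomeRange glider cosmetic.
# It wraps the shared MSS_Glider_Open graph, feeding four flap wave variations into its Open sounds slot and passing through pitch, volume and start-delay inputs at neutral defaults.
# The template's stereo audio and finished trigger go straight to the outputs.
# It uses glider open/close attenuation, the glider Land sound class and a concurrency override that blocks excess instances.
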MetaSoundSource /BRCosmetics/Sounds/Gliders/HomeRange/MSS_Glider_HomeRange_WingFlap_Big
inputs [
 PitchModulationRange.Max {type=Float}
 PitchModulationRange.Min {type=Float}
 StartDelay.Max {type=Time}
 StartDelay.Min {type=Time}
 TemplatePitchMultiplier {type=Float}
 TemplateVolumeMultiplier {type=Float}
 UE.Source.OnPlay {type=Trigger}
 VolumeModulationRange.Max {type=Float}
 VolumeModulationRange.Min {type=Float}
 WingFlapBig.Sounds {type=WaveAsset:Array}
]
interfaces [UE.OutputFormat.Stereo:1.0, UE.Source.OneShot:1.0, UE.Source:1.1]
outputs [
 "UE.OutputFormat.Stereo.Audio:0" {type=Audio}
 "UE.OutputFormat.Stereo.Audio:1" {type=Audio}
 UE.Source.OneShot.OnFinished {type=Trigger}
]
pages [Default {
  input_defaults {
   PitchModulationRange.Max=1
   PitchModulationRange.Min=1
   StartDelay.Max=0
   StartDelay.Min=0
   TemplatePitchMultiplier=1
   TemplateVolumeMultiplier=1
   UE.Source.OnPlay=null
   VolumeModulationRange.Max=1
   VolumeModulationRange.Min=1
   WingFlapBig.Sounds [
    &/Glider_HomeRange_WingFlap_Big_01
    &/Glider_HomeRange_WingFlap_Big_02
    &/Glider_HomeRange_WingFlap_Big_03
    &/Glider_HomeRange_WingFlap_Big_04
   ]
  }
  nodes [MSS_Glider_Open_0 = /FortMetasound/Cosmetics/Glider/MSS_Glider_Open(
    PitchModulationRange.Max=^PitchModulationRange.Max
    PitchModulationRange.Min=^PitchModulationRange.Min
    StartDelay.Max=^StartDelay.Max
    StartDelay.Min=^StartDelay.Min
    TemplatePitchMultiplier=^TemplatePitchMultiplier
    TemplateVolumeMultiplier=^TemplateVolumeMultiplier
    UE.Source.OnPlay=^UE.Source.OnPlay
    VolumeModulationRange.Max=^VolumeModulationRange.Max
    VolumeModulationRange.Min=^VolumeModulationRange.Min
    Open.Sounds=^WingFlapBig.Sounds
   )]
  links [
   {dst="@UE.OutputFormat.Stereo.Audio:0", src=MSS_Glider_Open_0.UE.OutputFormat.Stereo.Audio:0}
   {dst="@UE.OutputFormat.Stereo.Audio:1", src=MSS_Glider_Open_0.UE.OutputFormat.Stereo.Audio:1}
   {dst="@UE.Source.OneShot.OnFinished", src=MSS_Glider_Open_0.UE.Source.OneShot.OnFinished}
  ]
 }]
properties {
 assetImportData=&self:AssetImportData
 attenuationSettings=/Game/Sounds/Attenuation/Fort_Player_Glider_OpenClose_Athena_Att
 bOverrideConcurrency=true
 maxDistance=10000
 numChannels=2
 outputFormat=Stereo
 sampleRate=48000
 soundClassObject=&Fort_Gliders_Land
 volume=0.8
 concurrencyOverrides {maxCount=2, resolutionRule=PreventNew}
 concurrencySet [&Athena_Glider_Thrust_Conc]
}
